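# Physics asset for the Keyboard_Keys gadget, a cosmetic animation prop.
# It gives each of the thirteen key bones a uniform sphere body and the root one large, very short capsule that is close to a sphere.
# No constraints link the bodies, so the asset defines collision shapes without any joints between the keys and root.
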PhysicsAsset /BRCosmetics/Animation/Props/Keyboard_Keys/Meshes/Keyboard_Keys_Gadget_Physics
bodies {
 Key_01 [{shapeType=Sphere, radius=15, shapeName=Key_01_sphere}]
 Key_010 [{shapeType=Sphere, radius=15, shapeName=Key_010_sphere}]
 Key_011 [{shapeType=Sphere, radius=15, shapeName=Key_011_sphere}]
 Key_012 [{shapeType=Sphere, radius=15, shapeName=Key_012_sphere}]
 Key_013 [{shapeType=Sphere, radius=15, shapeName=Key_013_sphere}]
 Key_02 [{shapeType=Sphere, radius=15, shapeName=Key_02_sphere}]
 Key_03 [{shapeType=Sphere, radius=15, shapeName=Key_03_sphere}]
 Key_04 [{shapeType=Sphere, radius=15, shapeName=Key_04_sphere}]
 Key_05 [{shapeType=Sphere, radius=15, shapeName=Key_05_sphere}]
 Key_06 [{shapeType=Sphere, radius=15, shapeName=Key_06_sphere}]
 Key_07 [{shapeType=Sphere, radius=15, shapeName=Key_07_sphere}]
 Key_08 [{shapeType=Sphere, radius=15, shapeName=Key_08_sphere}]
 Key_09 [{shapeType=Sphere, radius=15, shapeName=Key_09_sphere}]
 root [{shapeType=Capsule, length=1, radius=76.505, shapeName=root_capsule, rotation=<rot 0 -90 0>}]
}
constraints []
